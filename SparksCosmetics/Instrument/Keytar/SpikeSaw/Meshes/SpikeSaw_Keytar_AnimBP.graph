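# Animation Blueprint for the SpikeSaw Keytar instrument cosmetic, layering an idle animation onto the keytar's skeleton.
# The AnimGraph takes an incoming Input Pose and uses Apply Additive to add the looping SpikeSaw_Keytar_Idle_Loop sequence on top before output.
# The EventGraph holds only unconnected default nodes and there are no variables, so no logic drives the blend.
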AnimBlueprint /SparksCosmetics/Instrument/Keytar/SpikeSaw/Meshes/SpikeSaw_Keytar_AnimBP
target_skeleton=&/SpikeSaw_Keytar_Skeleton
graphs [
 AnimGraph {nodes [Output Pose(Result=Apply Additive(
     Base=Input Pose()
     Additive="Sequence Player (additive) 'SpikeSaw_Keytar_Idle_Loop'"() {sequence=&/SpikeSaw_Keytar_Idle_Loop}
    ) {alphaBoolBlend {blendOption=Linear}})]}
 EventGraph {nodes [
   TryGetPawnOwner()
   Event BlueprintUpdateAnimation()
  ]}
]
variables []
